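# MetaSound source preset for the equip sound of the WaffleTruck launcher weapon, built on the shared MSS_Weapon_Equip parent.
# It supplies three equip wave variations, a master volume and a random pitch-shift range to the parent's graph, outputting mono one-shot audio.
# Weapon foley sound class, attenuation and concurrency settings govern its mixing and spatial falloff.
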MetaSoundSource /WeaponsUnvaulted/Assets/Audio/WaffleTruck_Launcher_DrunkenQuad/MSS_WaffleTruck_Launcher_Equip
preset_parent=/FortMetasound/Weapons/MSS_Weapon_Equip
inputs [
 Character.Affiliation {type=Enum:FortAudioAffiliation}
 Character.Health {type=Float}
 Character.UniqueID {type=Int32}
 MasterVolume {type=Float}
 Random.PitchShift.Base {type=Float}
 Random.PitchShift.Spread {type=Float}
 Sounds {type=WaveAsset:Array}
 UE.Source.OnPlay {type=Trigger}
]
interfaces [Fort.Character:1.0, UE.OutputFormat.Mono:1.0, UE.Source.OneShot:1.0, UE.Source:1.1]
outputs [
 "UE.OutputFormat.Mono.Audio:0" {type=Audio}
 UE.Source.OneShot.OnFinished {type=Trigger}
]
pages [Default {input_defaults {
   Character.Affiliation=0
   Character.Health=0
   Character.UniqueID=0
   MasterVolume=0.65
   Random.PitchShift.Base=0
   Random.PitchShift.Spread=0.5
   UE.Source.OnPlay=null
   Sounds [
    &/WaffleTruck_Launcher_Equip_03
    &/WaffleTruck_Launcher_Equip_02
    &/WaffleTruck_Launcher_Equip_01
   ]
  }}]
properties {
 assetImportData=&self:AssetImportData
 attenuationSettings=&AS_Fort_WeaponFoley
 maxDistance=2250
 sampleRate=48000
 soundClassObject=&Fort_WeaponFoley
 concurrencySet [&CS_Fort_WeaponFoley]
}
